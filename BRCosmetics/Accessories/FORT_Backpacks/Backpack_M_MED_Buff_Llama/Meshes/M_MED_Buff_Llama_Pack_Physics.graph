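# Physics asset for the Buff Llama backpack cosmetic, defining capsule bodies and constraints for its dangling bullets, pouches and pickaxe.
# The belt hangs off the pack bone with free rotation and parents seven bullets and two pouch masters, each pouch master carrying one child body.
# Bullets, pouch children and the two-capsule main pickaxe body swing mostly on one axis with minimal twist, while pouch masters and a second pickaxe body rotate freely.
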PhysicsAsset /BRCosmetics/Accessories/FORT_Backpacks/Backpack_M_MED_Buff_Llama/Meshes/M_MED_Buff_Llama_Pack_Physics
bodies {
 Dyn_Bullet1 [{shapeType=Capsule, length=3, radius=1, shapeName=Dyn_Bullet1_capsule, center=<1 0 0>}]
 Dyn_Bullet2 [{shapeType=Capsule, length=3, radius=1, shapeName=Dyn_Bullet2_capsule, center=<1 0 0>}]
 Dyn_Bullet3 [{shapeType=Capsule, length=3, radius=1, shapeName=Dyn_Bullet3_capsule, center=<0.5 0 0>}]
 Dyn_Bullet4 [{shapeType=Capsule, length=3, radius=1, shapeName=Dyn_Bullet4_capsule, center=<0.5 0 0>}]
 Dyn_Bullet5 [{shapeType=Capsule, length=3, radius=1, shapeName=Dyn_Bullet5_capsule, center=<0.5 0 0>}]
 Dyn_Bullet6 [{shapeType=Capsule, length=3, radius=1, shapeName=Dyn_Bullet6_capsule, center=<0.5 0 0>}]
 Dyn_Bullet7 [{shapeType=Capsule, length=3, radius=1, shapeName=Dyn_Bullet7_capsule, center=<0.5 0 0>}]
 Dyn_Master_Belt [{shapeType=Capsule, length=0.505, radius=0.505, shapeName=Dyn_Master_Belt_capsule}]
 Dyn_Master_PIckaxe [
  {shapeType=Capsule, length=5, radius=5, shapeName=Dyn_Master_PIckaxe_capsule, center=<0.9905 -0.0004 -0.7581>}
  {shapeType=Capsule, length=30, radius=1, shapeName=Dyn_Master_PIckaxe_capsule, center=<13.854 0 -0.4828>, rotation=<rot 90 0 0>}
 ]
 Dyn_Master_PIckaxe1 [{shapeType=Capsule, length=10, radius=3, shapeName=Dyn_Master_PIckaxe1_capsule, center=<-3.7123 0 -2.9932>, rotation=<rot 90 0 0>}]
 Dyn_Master_Pouch_l [{shapeType=Capsule, length=0.505, radius=0.505, shapeName=Dyn_Master_Pouch_l_capsule}]
 Dyn_Master_Pouch_r [{shapeType=Capsule, length=25, radius=2, shapeName=Dyn_Master_Pouch_r_capsule, center=<3.3745 1.6653 8.8651>}]
 Dyn_Pouch_l1 [{shapeType=Capsule, length=4, radius=2, shapeName=Dyn_Pouch_l1_capsule, center=<1 -1.6602 0>, rotation=<rot 90 0 0>}]
 Dyn_Pouch_r1 [{shapeType=Capsule, length=4, radius=2, shapeName=Dyn_Pouch_r1_capsule, center=<1.5 -1.6602 0>, rotation=<rot 90 0 0>}]
 pack [{shapeType=Capsule, length=1, radius=1, shapeName=pack_capsule}]
}
constraints [
 {bone1Name=Dyn_Bullet1, bone2Name=Dyn_Master_Belt, swing1LimitDegrees=20, swing2LimitDegrees=1, twistLimitDegrees=1}
 {bone1Name=Dyn_Bullet2, bone2Name=Dyn_Master_Belt, swing1LimitDegrees=20, swing2LimitDegrees=1, twistLimitDegrees=1}
 {bone1Name=Dyn_Bullet3, bone2Name=Dyn_Master_Belt, swing1LimitDegrees=20, swing2LimitDegrees=1, twistLimitDegrees=1}
 {bone1Name=Dyn_Bullet4, bone2Name=Dyn_Master_Belt, swing1LimitDegrees=20, swing2LimitDegrees=1, twistLimitDegrees=1}
 {bone1Name=Dyn_Bullet5, bone2Name=Dyn_Master_Belt, swing1LimitDegrees=20, swing2LimitDegrees=1, twistLimitDegrees=1}
 {bone1Name=Dyn_Bullet6, bone2Name=Dyn_Master_Belt, swing1LimitDegrees=20, swing2LimitDegrees=1, twistLimitDegrees=1}
 {bone1Name=Dyn_Bullet7, bone2Name=Dyn_Master_Belt, swing1LimitDegrees=20, swing2LimitDegrees=1, twistLimitDegrees=1}
 {
  bone1Name=Dyn_Master_Pouch_l
  bone2Name=Dyn_Master_Belt
  swing1LimitDegrees=45
  swing1Motion=Free
  swing2LimitDegrees=45
  swing2Motion=Free
  twistLimitDegrees=45
  twistMotion=Free
 }
 {
  bone1Name=Dyn_Master_Pouch_r
  bone2Name=Dyn_Master_Belt
  swing1LimitDegrees=45
  swing1Motion=Free
  swing2LimitDegrees=45
  swing2Motion=Free
  twistLimitDegrees=45
  twistMotion=Free
 }
 {bone1Name=Dyn_Pouch_l1, bone2Name=Dyn_Master_Pouch_l, swing1LimitDegrees=20, swing2LimitDegrees=10, twistLimitDegrees=1}
 {bone1Name=Dyn_Pouch_r1, bone2Name=Dyn_Master_Pouch_r, swing1LimitDegrees=20, swing2LimitDegrees=10, twistLimitDegrees=1}
 {
  bone1Name=Dyn_Master_Belt
  bone2Name=pack
  swing1LimitDegrees=45
  swing1Motion=Free
  swing2LimitDegrees=45
  swing2Motion=Free
  twistLimitDegrees=45
  twistMotion=Free
 }
 {
  bone1Name=Dyn_Master_PIckaxe1
  bone2Name=pack
  swing1LimitDegrees=45
  swing1Motion=Free
  swing2LimitDegrees=45
  swing2Motion=Free
  twistLimitDegrees=45
  twistMotion=Free
 }
 {bone1Name=Dyn_Master_PIckaxe, bone2Name=pack, swing1LimitDegrees=1, swing2LimitDegrees=30, twistLimitDegrees=1}
]
